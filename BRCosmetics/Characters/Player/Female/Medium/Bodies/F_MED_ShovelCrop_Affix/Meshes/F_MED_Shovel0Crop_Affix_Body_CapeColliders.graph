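# Physics asset of body colliders, named for cape simulation on the female medium ShovelCrop Affix player character body.
# Capsules cover the head, clavicles, spine, pelvis and feet, with spine_02 and spine_03 each pairing a short capsule with a longer one rotated 90 degrees.
# Arm, thigh, calf and neck bodies list no shapes.
# It has no constraints, so it provides collision volumes rather than a jointed ragdoll.
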PhysicsAsset /BRCosmetics/Characters/Player/Female/Medium/Bodies/F_MED_ShovelCrop_Affix/Meshes/F_MED_Shovel0Crop_Affix_Body_CapeColliders
bodies {
 calf_l []
 calf_r []
 clavicle_l [{shapeType=Capsule, length=7.6964, radius=8.2957, center=<6.1952 0.3666 -2.2105>, rotation=<rot -77.5565 177.3915 7.7315>}]
 clavicle_r [{shapeType=Capsule, length=7.6964, radius=8.2957, center=<-6.1952 -0.3666 2.2105>, rotation=<rot -77.5565 177.3915 -172.2685>}]
 foot_l [{shapeType=Capsule, length=12.5982, radius=6.5127, center=<-2.8504 4.3663 -0.3196>, rotation=<rot 0.0388 1.4701 -89.3982>}]
 foot_r [{shapeType=Capsule, length=12.5982, radius=6.5127, center=<2.8504 -4.3663 0.3196>, rotation=<rot 0.0388 1.4701 90.6018>}]
 head [{shapeType=Capsule, length=3, radius=15, center=<9.0697 0.0746 0>, rotation=<rot 0 -90 0>}]
 lowerarm_l []
 lowerarm_r []
 neck_02 []
 pelvis [{shapeType=Capsule, length=9.4028, radius=12.6596, center=<-2.2118 0.3944 -0.1981>}]
 spine_01 [{shapeType=Capsule, length=10.5171, radius=8.2354, center=<-2.8183 -1.0873 0>}]
 spine_02 [
  {shapeType=Capsule, length=5, radius=8.4141, center=<1.179 2.2974 0>}
  {shapeType=Capsule, length=20.6875, radius=9.3125, center=<-1.8317 2.606 0>, rotation=<rot 90 0 0>}
 ]
 spine_03 [
  {shapeType=Capsule, length=7.6348, radius=6.8829, center=<-0.0747 2.3725 0>}
  {shapeType=Capsule, length=18.9984, radius=11.0015, center=<-1.2326 2.6059 0>, rotation=<rot 90 0 0>}
 ]
 spine_04 [{shapeType=Capsule, length=3.5538, radius=12.6674, center=<5.5182 3.8381 0>}]
 spine_05 [{shapeType=Capsule, length=11.4859, radius=9.5791, center=<5.6377 1.6965 0>}]
 thigh_l []
 thigh_r []
 upperarm_l []
 upperarm_r []
}
constraints []
