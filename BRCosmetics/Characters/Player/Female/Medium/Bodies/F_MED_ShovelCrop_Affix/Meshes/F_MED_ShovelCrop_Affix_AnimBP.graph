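# Animation Blueprint for the F_MED_ShovelCrop_Affix female medium player body cosmetic.
# The AnimGraph copies the pose from the attached parent mesh, then runs it through the body's own Control Rig before output.
# Each animation update checks the current LOD and sets the Body_BS morph target fully on for LODs above 0, and off at LOD 0.
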AnimBlueprint /BRCosmetics/Characters/Player/Female/Medium/Bodies/F_MED_ShovelCrop_Affix/Meshes/F_MED_ShovelCrop_Affix_AnimBP
target_skeleton=&/F_MED_ShovelCrop_Affix_Skeleton
graphs [
 AnimGraph {nodes [
   n1 = Control Rig(Source=Copy Pose From Mesh() {bUseAttachedParent=true}) {controlRigAssetReference {blueprintRigClass=&/F_MED_ShovelCrop_Affix_CtrlRig.F_MED_ShovelCrop_Affix_CtrlRig_C}}
   Output Pose(Result=@n1) {layerGroup=None}
  ]}
 EventGraph {nodes [
   Event BlueprintUpdateAnimation()
   > n1 = Branch(Condition="integer > integer"(A=Get CurrentLOD(), B=0) {is=title K2Node_CallFunction})
   > SetMorphTarget!(execute, self:Anim Instance Object Reference, MorphTargetName:Name=Body_BS, Value:float=1) -> (then)
   SetMorphTarget(execute=@n1.else, MorphTargetName=Body_BS, Value=0)
  ]}
]
variables []
